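# Physics asset for the left-arm fur bones of the large male player body M_LRG_Rumble.
# Ten fur strands, A through J, each have a capsule on a MSTR bone and a capsule on a child bone offset along X.
# One constraint per strand joins the pair, with a wider first swing, a very narrow second swing and zero twist, so each strand flaps mostly in one plane.
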PhysicsAsset /Game/Characters/Player/Male/Large/Bodies/M_LRG_Rumble/Meshes/M_LRG_Rumble_ArmLFPhysics
bodies {
 dyn_armFurLfA1 [{shapeType=Capsule, length=1, radius=1, shapeName=dyn_armFurLfA1_capsule, center=<1.5 0 0>, rotation=<rot 90 0 0>}]
 dyn_armFurLfA_MSTR [{shapeType=Capsule, length=0.505, radius=0.505, shapeName=dyn_armFurLfA_MSTR_capsule, rotation=<rot 0 -90 0>}]
 dyn_armFurLfB1 [{shapeType=Capsule, length=1, radius=1, shapeName=dyn_armFurLfB1_capsule, center=<1.5 0 0>, rotation=<rot 90 0 0>}]
 dyn_armFurLfB_MSTR [{shapeType=Capsule, length=0.505, radius=0.505, shapeName=dyn_armFurLfB_MSTR_capsule, rotation=<rot 0 -90 0>}]
 dyn_armFurLfC1 [{shapeType=Capsule, length=1, radius=1, shapeName=dyn_armFurLfC1_capsule, center=<1.5 0 0>, rotation=<rot 90 0 0>}]
 dyn_armFurLfC_MSTR [{shapeType=Capsule, length=0.505, radius=0.505, shapeName=dyn_armFurLfC_MSTR_capsule, rotation=<rot 0 -90 0>}]
 dyn_armFurLfD1 [{shapeType=Capsule, length=1, radius=1, shapeName=dyn_armFurLfD1_capsule, center=<1.5 0 0>, rotation=<rot 90 0 0>}]
 dyn_armFurLfD_MSTR [{shapeType=Capsule, length=0.505, radius=0.505, shapeName=dyn_armFurLfD_MSTR_capsule, rotation=<rot 0 -90 0>}]
 dyn_armFurLfE1 [{shapeType=Capsule, length=1, radius=1, shapeName=dyn_armFurLfE1_capsule, center=<1.5 0 0>, rotation=<rot 90 0 0>}]
 dyn_armFurLfE_MSTR [{shapeType=Capsule, length=0.505, radius=0.505, shapeName=dyn_armFurLfE_MSTR_capsule, rotation=<rot 0 -90 0>}]
 dyn_armFurLfF1 [{shapeType=Capsule, length=1, radius=1, shapeName=dyn_armFurLfF1_capsule, center=<1.5 0 0>, rotation=<rot 90 0 0>}]
 dyn_armFurLfF_MSTR [{shapeType=Capsule, length=0.505, radius=0.505, shapeName=dyn_armFurLfF_MSTR_capsule, rotation=<rot 0 -90 0>}]
 dyn_armFurLfG1 [{shapeType=Capsule, length=1, radius=1, shapeName=dyn_armFurLfG1_capsule, center=<1.5 0 0>, rotation=<rot 90 0 0>}]
 dyn_armFurLfG_MSTR [{shapeType=Capsule, length=0.505, radius=0.505, shapeName=dyn_armFurLfG_MSTR_capsule, rotation=<rot 0 -90 0>}]
 dyn_armFurLfH1 [{shapeType=Capsule, length=1, radius=1, shapeName=dyn_armFurLfH1_capsule, center=<1.5 0 0>, rotation=<rot 90 0 0>}]
 dyn_armFurLfH_MSTR [{shapeType=Capsule, length=0.505, radius=0.505, shapeName=dyn_armFurLfH_MSTR_capsule, rotation=<rot 0 -90 0>}]
 dyn_armFurLfI1 [{shapeType=Capsule, length=1, radius=1, shapeName=dyn_armFurLfI1_capsule, center=<1.5 0 0>, rotation=<rot 90 0 0>}]
 dyn_armFurLfI_MSTR [{shapeType=Capsule, length=0.505, radius=0.505, shapeName=dyn_armFurLfI_MSTR_capsule, rotation=<rot 0 -90 0>}]
 dyn_armFurLfJ1 [{shapeType=Capsule, length=1, radius=1, shapeName=dyn_armFurLfJ1_capsule, center=<1.5 0 0>, rotation=<rot 90 0 0>}]
 dyn_armFurLfJ_MSTR [{shapeType=Capsule, length=0.505, radius=0.505, shapeName=dyn_armFurLfJ_MSTR_capsule, rotation=<rot 0 -90 0>}]
}
constraints [
 {bone1Name=dyn_armFurLfA1, bone2Name=dyn_armFurLfA_MSTR, swing1LimitDegrees=20, swing2LimitDegrees=2, twistLimitDegrees=0}
 {bone1Name=dyn_armFurLfB1, bone2Name=dyn_armFurLfB_MSTR, swing1LimitDegrees=20, swing2LimitDegrees=2, twistLimitDegrees=0}
 {bone1Name=dyn_armFurLfC1, bone2Name=dyn_armFurLfC_MSTR, swing1LimitDegrees=20, swing2LimitDegrees=2, twistLimitDegrees=0}
 {bone1Name=dyn_armFurLfD1, bone2Name=dyn_armFurLfD_MSTR, swing1LimitDegrees=20, swing2LimitDegrees=2, twistLimitDegrees=0}
 {bone1Name=dyn_armFurLfE1, bone2Name=dyn_armFurLfE_MSTR, swing1LimitDegrees=20, swing2LimitDegrees=2, twistLimitDegrees=0}
 {bone1Name=dyn_armFurLfF1, bone2Name=dyn_armFurLfF_MSTR, swing1LimitDegrees=20, swing2LimitDegrees=2, twistLimitDegrees=0}
 {bone1Name=dyn_armFurLfG1, bone2Name=dyn_armFurLfG_MSTR, swing1LimitDegrees=20, swing2LimitDegrees=2, twistLimitDegrees=0}
 {bone1Name=dyn_armFurLfH1, bone2Name=dyn_armFurLfH_MSTR, swing1LimitDegrees=20, swing2LimitDegrees=2, twistLimitDegrees=0}
 {bone1Name=dyn_armFurLfI1, bone2Name=dyn_armFurLfI_MSTR, swing1LimitDegrees=20, swing2LimitDegrees=2, twistLimitDegrees=0}
 {bone1Name=dyn_armFurLfJ1, bone2Name=dyn_armFurLfJ_MSTR, swing1LimitDegrees=20, swing2LimitDegrees=2, twistLimitDegrees=0}
]
