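# Physics asset for the right-arm fur bones on the large male Rumble player body.
# Ten separate two-body chains (A to J) each pair a small MSTR capsule with a larger, offset capsule on the fur bone.
# Each pair's constraint has a zero twist limit and uneven swing limits, favoring one axis, with no links between chains.
# The fur-bone capsules all share a left-arm shape name, dyn_armFurLfC1_capsule.
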PhysicsAsset /Game/Characters/Player/Male/Large/Bodies/M_LRG_Rumble/Meshes/M_LRG_Rumble_ArmRTPhysics
bodies {
 dyn_armFurRtA1 [{shapeType=Capsule, length=1, radius=1, shapeName=dyn_armFurLfC1_capsule, center=<1.5 0 0>, rotation=<rot 90 0 0>}]
 dyn_armFurRtA_MSTR [{shapeType=Capsule, length=0.505, radius=0.505, shapeName=dyn_armFurRtA_MSTR_capsule, rotation=<rot 0 -90 0>}]
 dyn_armFurRtB1 [{shapeType=Capsule, length=1, radius=1, shapeName=dyn_armFurLfC1_capsule, center=<1.5 0 0>, rotation=<rot 90 0 0>}]
 dyn_armFurRtB_MSTR [{shapeType=Capsule, length=0.505, radius=0.505, shapeName=dyn_armFurRtB_MSTR_capsule, rotation=<rot 0 -90 0>}]
 dyn_armFurRtC1 [{shapeType=Capsule, length=1, radius=1, shapeName=dyn_armFurLfC1_capsule, center=<1.5 0 0>, rotation=<rot 90 0 0>}]
 dyn_armFurRtC_MSTR [{shapeType=Capsule, length=0.505, radius=0.505, shapeName=dyn_armFurRtC_MSTR_capsule, rotation=<rot 0 -90 0>}]
 dyn_armFurRtD1 [{shapeType=Capsule, length=1, radius=1, shapeName=dyn_armFurLfC1_capsule, center=<1.5 0 0>, rotation=<rot 90 0 0>}]
 dyn_armFurRtD_MSTR [{shapeType=Capsule, length=0.505, radius=0.505, shapeName=dyn_armFurRtD_MSTR_capsule, rotation=<rot 0 -90 0>}]
 dyn_armFurRtE1 [{shapeType=Capsule, length=1, radius=1, shapeName=dyn_armFurLfC1_capsule, center=<1.5 0 0>, rotation=<rot 90 0 0>}]
 dyn_armFurRtE_MSTR [{shapeType=Capsule, length=0.505, radius=0.505, shapeName=dyn_armFurRtE_MSTR_capsule, rotation=<rot 0 -90 0>}]
 dyn_armFurRtF1 [{shapeType=Capsule, length=1, radius=1, shapeName=dyn_armFurLfC1_capsule, center=<1.5 0 0>, rotation=<rot 90 0 0>}]
 dyn_armFurRtF_MSTR [{shapeType=Capsule, length=0.505, radius=0.505, shapeName=dyn_armFurRtF_MSTR_capsule, rotation=<rot 0 -90 0>}]
 dyn_armFurRtG1 [{shapeType=Capsule, length=1, radius=1, shapeName=dyn_armFurLfC1_capsule, center=<1.5 0 0>, rotation=<rot 90 0 0>}]
 dyn_armFurRtG_MSTR [{shapeType=Capsule, length=0.505, radius=0.505, shapeName=dyn_armFurRtG_MSTR_capsule, rotation=<rot 0 -90 0>}]
 dyn_armFurRtH1 [{shapeType=Capsule, length=1, radius=1, shapeName=dyn_armFurLfC1_capsule, center=<1.5 0 0>, rotation=<rot 90 0 0>}]
 dyn_armFurRtH_MSTR [{shapeType=Capsule, length=0.505, radius=0.505, shapeName=dyn_armFurRtH_MSTR_capsule, rotation=<rot 0 -90 0>}]
 dyn_armFurRtI1 [{shapeType=Capsule, length=1, radius=1, shapeName=dyn_armFurLfC1_capsule, center=<1.5 0 0>, rotation=<rot 90 0 0>}]
 dyn_armFurRtI_MSTR [{shapeType=Capsule, length=0.505, radius=0.505, shapeName=dyn_armFurRtI_MSTR_capsule, rotation=<rot 0 -90 0>}]
 dyn_armFurRtJ1 [{shapeType=Capsule, length=1, radius=1, shapeName=dyn_armFurLfC1_capsule, center=<1.5 0 0>, rotation=<rot 90 0 0>}]
 dyn_armFurRtJ_MSTR [{shapeType=Capsule, length=0.505, radius=0.505, shapeName=dyn_armFurRtJ_MSTR_capsule, rotation=<rot 0 -90 0>}]
}
constraints [
 {bone1Name=dyn_armFurRtJ1, bone2Name=dyn_armFurRtJ_MSTR, swing1LimitDegrees=20, swing2LimitDegrees=2, twistLimitDegrees=0}
 {bone1Name=dyn_armFurRtI1, bone2Name=dyn_armFurRtI_MSTR, swing1LimitDegrees=20, swing2LimitDegrees=2, twistLimitDegrees=0}
 {bone1Name=dyn_armFurRtH1, bone2Name=dyn_armFurRtH_MSTR, swing1LimitDegrees=20, swing2LimitDegrees=2, twistLimitDegrees=0}
 {bone1Name=dyn_armFurRtG1, bone2Name=dyn_armFurRtG_MSTR, swing1LimitDegrees=20, swing2LimitDegrees=2, twistLimitDegrees=0}
 {bone1Name=dyn_armFurRtF1, bone2Name=dyn_armFurRtF_MSTR, swing1LimitDegrees=20, swing2LimitDegrees=2, twistLimitDegrees=0}
 {bone1Name=dyn_armFurRtE1, bone2Name=dyn_armFurRtE_MSTR, swing1LimitDegrees=20, swing2LimitDegrees=2, twistLimitDegrees=0}
 {bone1Name=dyn_armFurRtD1, bone2Name=dyn_armFurRtD_MSTR, swing1LimitDegrees=20, swing2LimitDegrees=2, twistLimitDegrees=0}
 {bone1Name=dyn_armFurRtC1, bone2Name=dyn_armFurRtC_MSTR, swing1LimitDegrees=20, swing2LimitDegrees=2, twistLimitDegrees=0}
 {bone1Name=dyn_armFurRtB1, bone2Name=dyn_armFurRtB_MSTR, swing1LimitDegrees=20, swing2LimitDegrees=2, twistLimitDegrees=0}
 {bone1Name=dyn_armFurRtA1, bone2Name=dyn_armFurRtA_MSTR, swing1LimitDegrees=20, swing2LimitDegrees=2, twistLimitDegrees=0}
]
