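# Physics asset for the left chain on the large male player body Rumble, covering the dyn_chain_l bone and the left calf twist bone.
# Each of the two bones has a single capsule body, with the chain capsule offset along its local X axis.
# One constraint joins the chain to the calf, with a narrow limit on one swing axis, a much wider one on the other, and a zero twist limit.
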PhysicsAsset /Game/Characters/Player/Male/Large/Bodies/M_LRG_Rumble/Meshes/M_LRG_Rumble_ChainLFPhysics
bodies {
 calf_twist_01_l [{shapeType=Capsule, length=1, radius=1, shapeName=calf_twist_01_l_capsule, rotation=<rot 90 0 0>}]
 dyn_chain_l [{shapeType=Capsule, length=1, radius=1, shapeName=dyn_chain_l_capsule, center=<1.5 0 0>, rotation=<rot 90 0 0>}]
}
constraints [{bone1Name=dyn_chain_l, bone2Name=calf_twist_01_l, swing1LimitDegrees=3, swing2LimitDegrees=45, twistLimitDegrees=0}]
