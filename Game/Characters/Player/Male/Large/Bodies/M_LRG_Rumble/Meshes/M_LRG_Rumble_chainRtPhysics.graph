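# Physics asset for the right-side chain on the large male Rumble player body, covering the dyn_chain_r bone and the right calf twist bone.
# A small capsule on the chain bone and a larger capsule on the calf are joined by one constraint.
# Its limits allow swing mostly on one axis, with little on the other and a zero twist limit.
# The chain capsule carries a left-side shape name despite sitting on the right bone.
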PhysicsAsset /Game/Characters/Player/Male/Large/Bodies/M_LRG_Rumble/Meshes/M_LRG_Rumble_chainRtPhysics
bodies {
 calf_twist_01_r [{
   shapeType=Capsule
   length=11.753
   radius=9.411
   shapeName=calf_twist_01_r_capsule
   center=<-2.7794 0.347 1.1617>
   rotation=<rot 0 -12.7666 -50.4935>
  }]
 dyn_chain_r [{shapeType=Capsule, length=1, radius=1, shapeName=dyn_chain_l_capsule, center=<1.5 0 0>, rotation=<rot 90 0 0>}]
}
constraints [{bone1Name=dyn_chain_r, bone2Name=calf_twist_01_r, swing1LimitDegrees=45, swing2LimitDegrees=3, twistLimitDegrees=0}]
