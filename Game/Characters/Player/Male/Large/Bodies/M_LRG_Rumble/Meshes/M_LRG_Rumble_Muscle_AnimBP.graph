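# Post-process style Anim Blueprint for the large male Rumble body, adding deformation, secondary motion and muscle flexing.
# AnimGraph copies the parent mesh pose through a deform Control Rig, positions arm-fur bones with weighted constraints, then simulates arm and calf-chain RigidBodies using overridden gravity.
# When muscles are enabled at LOD 1 or below, EventGraph converts joint rotation deltas through an easing curve into material muscle parameters.
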AnimBlueprint /Game/Characters/Player/Male/Large/Bodies/M_LRG_Rumble/Meshes/M_LRG_Rumble_Muscle_AnimBP
target_skeleton=&/M_LRG_Rumble_Skeleton
graphs [
 AnimGraph {nodes [
   n1 = Control Rig(Source=Copy Pose From Mesh() {bCopyCurves=true, bUseAttachedParent=true}) {
    alphaBoolBlend {blendOption=Linear}
    controlRigAssetReference {blueprintRigClass=&/M_LRG_Rumble_DeformRig_Master_ControlRig.M_LRG_Rumble_DeformRig_Master_ControlRig_C}
   }
   n2 = "Constraint - dyn_armFurLfH_MSTR (PPPPP)"!(
    ComponentPose=Local To Component(LocalPose=@n1)
    Alpha=1
    ConstraintWeights_0=0.0364
    ConstraintWeights_1:float=0.2726
    ConstraintWeights_2:float=0.1477
    ConstraintWeights_3:float=0.5209
    ConstraintWeights_4:float=0.0223
   ) -> (Pose) {
    alphaBoolBlend {blendOption=Linear}
    boneToModify {boneName=dyn_armFurLfH_MSTR}
    constraintSetup [
     {targetBone {boneName=upperarm_l}}
     {targetBone {boneName=lowerarm_l}}
     {targetBone {boneName=lowerarm_twist_01_l}}
     {targetBone {boneName=lowerarm_twist_02_l}}
     {targetBone {boneName=deform_elbow_out_l}}
    ]
    constraintWeights [1,1,1,1,1]
   }
   n3 = "Constraint - dyn_armFurRtH_MSTR (PPPPP)"!(
    ComponentPose=@n2
    Alpha=1
    ConstraintWeights_0=0.0364
    ConstraintWeights_1:float=0.2726
    ConstraintWeights_2:float=0.1477
    ConstraintWeights_3:float=0.5209
    ConstraintWeights_4:float=0.0223
   ) -> (Pose) {
    alphaBoolBlend {blendOption=Linear}
    boneToModify {boneName=dyn_armFurRtH_MSTR}
    constraintSetup [
     {targetBone {boneName=upperarm_r}}
     {targetBone {boneName=lowerarm_r}}
     {targetBone {boneName=lowerarm_twist_01_r}}
     {targetBone {boneName=lowerarm_twist_02_r}}
     {targetBone {boneName=deform_elbow_out_r}}
    ]
    constraintWeights [1,1,1,1,1]
   }
   n4 = "Constraint - dyn_armFurLfE_MSTR (PPPP)"!(
    ComponentPose=@n3
    Alpha=1
    ConstraintWeights_0=0.0165
    ConstraintWeights_1:float=0.2255
    ConstraintWeights_2:float=0.2135
    ConstraintWeights_3:float=0.5446
   ) -> (Pose) {
    alphaBoolBlend {blendOption=Linear}
    boneToModify {boneName=dyn_armFurLfE_MSTR}
    constraintSetup [
     {targetBone {boneName=upperarm_l}}
     {targetBone {boneName=lowerarm_l}}
     {targetBone {boneName=lowerarm_twist_01_l}}
     {targetBone {boneName=lowerarm_twist_02_l}}
    ]
    constraintWeights [1,1,1,1]
   }
   n5 = "Constraint - dyn_armFurRtE_MSTR (PPPP)"!(
    ComponentPose=@n4
    Alpha=1
    ConstraintWeights_0=0.0165
    ConstraintWeights_1:float=0.2255
    ConstraintWeights_2:float=0.2135
    ConstraintWeights_3:float=0.5446
   ) -> (Pose) {
    alphaBoolBlend {blendOption=Linear}
    boneToModify {boneName=dyn_armFurRtE_MSTR}
    constraintSetup [
     {targetBone {boneName=upperarm_r}}
     {targetBone {boneName=lowerarm_r}}
     {targetBone {boneName=lowerarm_twist_01_r}}
     {targetBone {boneName=lowerarm_twist_02_r}}
    ]
    constraintWeights [1,1,1,1]
   }
   n6 = "Constraint - dyn_armFurLfC_MSTR (PPPP)"!(
    ComponentPose=@n5
    Alpha=1
    ConstraintWeights_0=0.0038
    ConstraintWeights_1:float=0.1637
    ConstraintWeights_2:float=0.326
    ConstraintWeights_3:float=0.5065
   ) -> (Pose) {
    alphaBoolBlend {blendOption=Linear}
    boneToModify {boneName=dyn_armFurLfC_MSTR}
    constraintSetup [
     {targetBone {boneName=upperarm_l}}
     {targetBone {boneName=lowerarm_l}}
     {targetBone {boneName=lowerarm_twist_01_l}}
     {targetBone {boneName=lowerarm_twist_02_l}}
    ]
    constraintWeights [1,1,1,1]
   }
   n7 = "Constraint - dyn_armFurRtC_MSTR (PPPP)"!(
    ComponentPose=@n6
    Alpha=1
    ConstraintWeights_0=0.0038
    ConstraintWeights_1:float=0.1637
    ConstraintWeights_2:float=0.326
    ConstraintWeights_3:float=0.5065
   ) -> (Pose) {
    alphaBoolBlend {blendOption=Linear}
    boneToModify {boneName=dyn_armFurRtC_MSTR}
    constraintSetup [
     {targetBone {boneName=upperarm_r}}
     {targetBone {boneName=lowerarm_r}}
     {targetBone {boneName=lowerarm_twist_01_r}}
     {targetBone {boneName=lowerarm_twist_02_r}}
    ]
    constraintWeights [1,1,1,1]
   }
   n8 = "Constraint - dyn_armFurLfA_MSTR (PPPP)"!(
    ComponentPose=@n7
    Alpha=1
    ConstraintWeights_0=0.1073
    ConstraintWeights_1:float=0.0088
    ConstraintWeights_2:float=0.4675
    ConstraintWeights_3:float=0.4164
   ) -> (Pose) {
    alphaBoolBlend {blendOption=Linear}
    boneToModify {boneName=dyn_armFurLfA_MSTR}
    constraintSetup [
     {targetBone {boneName=lowerarm_l}}
     {targetBone {boneName=deform_wrist_up_l}}
     {targetBone {boneName=lowerarm_twist_01_l}}
     {targetBone {boneName=lowerarm_twist_02_l}}
    ]
    constraintWeights [1,1,1,1]
   }
   n9 = "Constraint - dyn_armFurRtA_MSTR (PPPP)"!(
    ComponentPose=@n8
    Alpha=1
    ConstraintWeights_0=0.1073
    ConstraintWeights_1:float=0.0088
    ConstraintWeights_2:float=0.4675
    ConstraintWeights_3:float=0.4164
   ) -> (Pose) {
    alphaBoolBlend {blendOption=Linear}
    boneToModify {boneName=dyn_armFurRtA_MSTR}
    constraintSetup [
     {targetBone {boneName=lowerarm_r}}
     {targetBone {boneName=deform_wrist_up_r}}
     {targetBone {boneName=lowerarm_twist_01_r}}
     {targetBone {boneName=lowerarm_twist_02_r}}
    ]
    constraintWeights [1,1,1,1]
   }
   n10 = "Constraint - dyn_armFurLfI_MSTR (PPPPP)"!(
    ComponentPose=@n9
    Alpha=1
    ConstraintWeights_0=0.0364
    ConstraintWeights_1:float=0.2726
    ConstraintWeights_2:float=0.1477
    ConstraintWeights_3:float=0.5209
    ConstraintWeights_4:float=0.0223
   ) -> (Pose) {
    alphaBoolBlend {blendOption=Linear}
    boneToModify {boneName=dyn_armFurLfI_MSTR}
    constraintSetup [
     {targetBone {boneName=upperarm_l}}
     {targetBone {boneName=lowerarm_l}}
     {targetBone {boneName=lowerarm_twist_01_l}}
     {targetBone {boneName=lowerarm_twist_02_l}}
     {targetBone {boneName=deform_elbow_out_l}}
    ]
    constraintWeights [1,1,1,1,1]
   }
   n11 = "Constraint - dyn_armFurRtI_MSTR (PPPPP)"!(
    ComponentPose=@n10
    Alpha=1
    ConstraintWeights_0=0.0586
    ConstraintWeights_1:float=0.298
    ConstraintWeights_2:float=0.1011
    ConstraintWeights_3:float=0.4697
    ConstraintWeights_4:float=0.0726
   ) -> (Pose) {
    alphaBoolBlend {blendOption=Linear}
    boneToModify {boneName=dyn_armFurRtI_MSTR}
    constraintSetup [
     {targetBone {boneName=upperarm_r}}
     {targetBone {boneName=lowerarm_r}}
     {targetBone {boneName=lowerarm_twist_01_r}}
     {targetBone {boneName=lowerarm_twist_02_r}}
     {targetBone {boneName=deform_elbow_out_r}}
    ]
    constraintWeights [1,1,1,1,1]
   }
   n12 = "Constraint - dyn_armFurLfD_MSTR (PPPPP)"!(
    ComponentPose=@n11
    Alpha=1
    ConstraintWeights_0=0.0063
    ConstraintWeights_1:float=0.2151
    ConstraintWeights_2:float=0.2147
    ConstraintWeights_3:float=0.5601
    ConstraintWeights_4:float=0.0038
   ) -> (Pose) {
    alphaBoolBlend {blendOption=Linear}
    boneToModify {boneName=dyn_armFurLfD_MSTR}
    constraintSetup [
     {targetBone {boneName=upperarm_l}}
     {targetBone {boneName=lowerarm_l}}
     {targetBone {boneName=lowerarm_twist_01_l}}
     {targetBone {boneName=lowerarm_twist_02_l}}
     {targetBone {boneName=deform_elbow_out_l}}
    ]
    constraintWeights [1,1,1,1,1]
   }
   n13 = "Constraint - dyn_armFurRtD_MSTR (PPPPP)"!(
    ComponentPose=@n12
    Alpha=1
    ConstraintWeights_0=0.0063
    ConstraintWeights_1:float=0.2151
    ConstraintWeights_2:float=0.2147
    ConstraintWeights_3:float=0.5601
    ConstraintWeights_4:float=0.0038
   ) -> (Pose) {
    alphaBoolBlend {blendOption=Linear}
    boneToModify {boneName=dyn_armFurRtD_MSTR}
    constraintSetup [
     {targetBone {boneName=upperarm_r}}
     {targetBone {boneName=lowerarm_r}}
     {targetBone {boneName=lowerarm_twist_01_r}}
     {targetBone {boneName=lowerarm_twist_02_r}}
     {targetBone {boneName=deform_elbow_out_r}}
    ]
    constraintWeights [1,1,1,1,1]
   }
   n14 = "Constraint - dyn_armFurLfJ_MSTR (PPPPPP)"!(
    ComponentPose=@n13
    Alpha=1
    ConstraintWeights_0=0.0799
    ConstraintWeights_1:float=0.2836
    ConstraintWeights_2:float=0.0585
    ConstraintWeights_3:float=0.3704
    ConstraintWeights_4:float=0.2014
    ConstraintWeights_5:float=0.0063
   ) -> (Pose) {
    alphaBoolBlend {blendOption=Linear}
    boneToModify {boneName=dyn_armFurLfJ_MSTR}
    constraintSetup [
     {targetBone {boneName=upperarm_l}}
     {targetBone {boneName=lowerarm_l}}
     {targetBone {boneName=lowerarm_twist_01_l}}
     {targetBone {boneName=lowerarm_twist_02_l}}
     {targetBone {boneName=deform_elbow_out_l}}
     {targetBone {boneName=upperarm_twist_02_l}}
    ]
    constraintWeights [1,1,1,1,1,1]
   }
   n15 = "Constraint - dyn_armFurRtJ_MSTR (PPPPPP)"!(
    ComponentPose=@n14
    Alpha=1
    ConstraintWeights_0=0.0799
    ConstraintWeights_1:float=0.2836
    ConstraintWeights_2:float=0.0585
    ConstraintWeights_3:float=0.3704
    ConstraintWeights_4:float=0.2014
    ConstraintWeights_5:float=0.0063
   ) -> (Pose) {
    alphaBoolBlend {blendOption=Linear}
    boneToModify {boneName=dyn_armFurRtJ_MSTR}
    constraintSetup [
     {targetBone {boneName=upperarm_r}}
     {targetBone {boneName=lowerarm_r}}
     {targetBone {boneName=lowerarm_twist_01_r}}
     {targetBone {boneName=lowerarm_twist_02_r}}
     {targetBone {boneName=deform_elbow_out_r}}
     {targetBone {boneName=upperarm_twist_02_r}}
    ]
    constraintWeights [1,1,1,1,1,1]
   }
   n16 = "Constraint - dyn_armFurLfG_MSTR (PPPPP)"!(
    ComponentPose=@n15
    Alpha=1
    ConstraintWeights_0=0.0298
    ConstraintWeights_1:float=0.2644
    ConstraintWeights_2:float=0.1488
    ConstraintWeights_3:float=0.5319
    ConstraintWeights_4:float=0.0251
   ) -> (Pose) {
    alphaBoolBlend {blendOption=Linear}
    boneToModify {boneName=dyn_armFurLfG_MSTR}
    constraintSetup [
     {targetBone {boneName=upperarm_l}}
     {targetBone {boneName=lowerarm_l}}
     {targetBone {boneName=lowerarm_twist_01_l}}
     {targetBone {boneName=lowerarm_twist_02_l}}
     {targetBone {boneName=deform_elbow_out_l}}
    ]
    constraintWeights [1,1,1,1,1]
   }
   n17 = "Constraint - dyn_armFurRtG_MSTR (PPPPP)"!(
    ComponentPose=@n16
    Alpha=1
    ConstraintWeights_0=0.0534
    ConstraintWeights_1:float=0.2931
    ConstraintWeights_2:float=0.0976
    ConstraintWeights_3:float=0.4721
    ConstraintWeights_4:float=0.0838
   ) -> (Pose) {
    alphaBoolBlend {blendOption=Linear}
    boneToModify {boneName=dyn_armFurRtG_MSTR}
    constraintSetup [
     {targetBone {boneName=upperarm_r}}
     {targetBone {boneName=lowerarm_r}}
     {targetBone {boneName=lowerarm_twist_01_r}}
     {targetBone {boneName=lowerarm_twist_02_r}}
     {targetBone {boneName=deform_elbow_out_r}}
    ]
    constraintWeights [1,1,1,1,1]
   }
   n18 = "Constraint - dyn_armFurLfF_MSTR (PPP)"!(ComponentPose=@n17, Alpha=1, ConstraintWeights_0=0.1602, ConstraintWeights_1:float=0.3262, ConstraintWeights_2:float=0.5136) -> (Pose) {
    alphaBoolBlend {blendOption=Linear}
    boneToModify {boneName=dyn_armFurLfF_MSTR}
    constraintSetup [{targetBone {boneName=lowerarm_l}}, {targetBone {boneName=lowerarm_twist_01_l}}, {targetBone {boneName=lowerarm_twist_02_l}}]
    constraintWeights [1,1,1]
   }
   n19 = "Constraint - dyn_armFurRtF_MSTR (PPP)"!(ComponentPose=@n18, Alpha=1, ConstraintWeights_0=0.1631, ConstraintWeights_1:float=0.3293, ConstraintWeights_2:float=0.5076) -> (Pose) {
    alphaBoolBlend {blendOption=Linear}
    boneToModify {boneName=dyn_armFurRtF_MSTR}
    constraintSetup [{targetBone {boneName=lowerarm_r}}, {targetBone {boneName=lowerarm_twist_01_r}}, {targetBone {boneName=lowerarm_twist_02_r}}]
    constraintWeights [1,1,1]
   }
   n20 = "Constraint - dyn_armFurLfB_MSTR (PPPP)"!(
    ComponentPose=@n19
    Alpha=1
    ConstraintWeights_0=0.1076
    ConstraintWeights_1:float=0.0058
    ConstraintWeights_2:float=0.4706
    ConstraintWeights_3:float=0.4161
   ) -> (Pose) {
    alphaBoolBlend {blendOption=Linear}
    boneToModify {boneName=dyn_armFurLfB_MSTR}
    constraintSetup [
     {targetBone {boneName=lowerarm_l}}
     {targetBone {boneName=deform_wrist_up_l}}
     {targetBone {boneName=lowerarm_twist_01_l}}
     {targetBone {boneName=lowerarm_twist_02_l}}
    ]
    constraintWeights [1,1,1,1]
   }
   n21 = "Constraint - dyn_armFurRtB_MSTR (PPPP)"!(
    ComponentPose=@n20
    Alpha=1
    ConstraintWeights_0=0.1076
    ConstraintWeights_1:float=0.0058
    ConstraintWeights_2:float=0.4706
    ConstraintWeights_3:float=0.4161
   ) -> (Pose) {
    alphaBoolBlend {blendOption=Linear}
    boneToModify {boneName=dyn_armFurRtB_MSTR}
    constraintSetup [
     {targetBone {boneName=lowerarm_r}}
     {targetBone {boneName=deform_wrist_up_r}}
     {targetBone {boneName=lowerarm_twist_01_r}}
     {targetBone {boneName=lowerarm_twist_02_r}}
    ]
    constraintWeights [1,1,1,1]
   }
   n22 = Get GravityOverride()
   n23 = RigidBody!(ComponentPose=@n21, Alpha, OverrideWorldGravity:Vector=@n22, ExternalForce) -> (Pose) {
    bClampLinearTranslationLimitToRefPose=true
    bOverrideWorldGravity=true
    overridePhysicsAsset=&/M_LRG_Rumble_ArmLFPhysics
    simulationSpace=BaseBoneSpace
    alphaBoolBlend {blendOption=Linear}
    baseBoneRef {boneName=lowerarm_l}
    componentAppliedLinearAccClamp=<3000 3000 3000>
    componentLinearAccScale=<0.4 0.25 0.25>
    componentLinearVelScale=<0.4 0.25 0.25>
    overrideWorldGravity {z=-980}
    simSpaceSettings {
     maxAngularAcceleration=1000
     maxAngularVelocity=3
     maxLinearAcceleration=700
     maxLinearVelocity=500
     velocityScaleZ=0
     worldAlpha=1
     externalLinearDragV=<0.2 0.2>
    }
   }
   n24 = RigidBody(ComponentPose=@n23, Alpha=0.75, OverrideWorldGravity=@n22) {
    bClampLinearTranslationLimitToRefPose=true
    bOverrideWorldGravity=true
    overridePhysicsAsset=&/M_LRG_Rumble_ArmRTPhysics
    simulationSpace=BaseBoneSpace
    alphaBoolBlend {blendOption=Linear}
    baseBoneRef {boneName=lowerarm_r}
    componentAppliedLinearAccClamp=<3000 3000 3000>
    componentLinearAccScale=<0.4 0.25 0.25>
    componentLinearVelScale=<0.4 0.25 0.25>
    overrideWorldGravity {z=-980}
    simSpaceSettings {
     maxAngularAcceleration=1000
     maxAngularVelocity=3
     maxLinearAcceleration=700
     maxLinearVelocity=500
     velocityScaleZ=0
     worldAlpha=1
     externalLinearDragV=<0.2 0.2>
    }
   }
   n25 = RigidBody(ComponentPose=@n24, OverrideWorldGravity=@n22) {
    bClampLinearTranslationLimitToRefPose=true
    bOverrideWorldGravity=true
    overridePhysicsAsset=&/M_LRG_Rumble_ChainLFPhysics
    simulationSpace=BaseBoneSpace
    alphaBoolBlend {blendOption=Linear}
    baseBoneRef {boneName=calf_twist_01_l}
    componentAppliedLinearAccClamp=<3000 3000 3000>
    componentLinearAccScale=<0.5 0.3 0.3>
    componentLinearVelScale=<0.5 0.3 0.3>
    overrideWorldGravity {z=-980}
    simSpaceSettings {
     maxAngularAcceleration=1000
     maxAngularVelocity=3
     maxLinearAcceleration=700
     maxLinearVelocity=500
     velocityScaleZ=0
     worldAlpha=1
     externalLinearDragV=<0.2 0.2>
    }
   }
   n26 = RigidBody(ComponentPose=@n25, OverrideWorldGravity=@n22) {
    bClampLinearTranslationLimitToRefPose=true
    bOverrideWorldGravity=true
    overridePhysicsAsset=&/M_LRG_Rumble_chainRtPhysics
    simulationSpace=BaseBoneSpace
    alphaBoolBlend {blendOption=Linear}
    baseBoneRef {boneName=calf_twist_01_r}
    componentAppliedLinearAccClamp=<3000 3000 3000>
    componentLinearAccScale=<0.5 0.3 0.3>
    componentLinearVelScale=<0.5 0.3 0.3>
    overrideWorldGravity {z=-980}
    simSpaceSettings {
     maxAngularAcceleration=1000
     maxAngularVelocity=3
     maxLinearAcceleration=700
     maxLinearVelocity=500
     velocityScaleZ=0
     worldAlpha=1
     externalLinearDragV=<0.2 0.2>
    }
   }
   Output Pose(Result=Component To Local(ComponentPose=@n26)) {layerGroup=None}
   Sequence Player 'M_LRG_ROM_Anim'() {sequence=&/M_LRG_ROM_Anim}
  ]}
 EventGraph {nodes [
   Event BlueprintUpdateAnimation()
   > n1 = Sequence!(execute) -> (then_0, then_1, then_2:Exec, then_3:Exec, then_4:Exec)
   n2 = Get CurrentLOD()
   > Branch(Condition=AND Boolean(A=Get IsMuscleON() -> (IsMuscleON:bool), B="integer <= integer"(A=@n2, B=1)))
   > n3 = Sequence()
   > n4 = Sequence()
   Set wrist_bend_l(
    execute=@n4.then_3
    wrist_bend_l:double=GetAlphaFromReferenceTransformDelta(SkeletalMeshComponent=GetOwnerSkeletalMeshComponent(), BoneName=hand_l, BaseName=lowerarm_l, InputRange=(X=0,Y=1), TwistAxis="1,0,0").RawOutput
   ) -> (Output_Get:double)
   > Set wrist_bend_r(wrist_bend_r:double=GetAlphaFromReferenceTransformDelta(SkeletalMeshComponent=GetOwnerSkeletalMeshComponent(), BoneName=hand_r, BaseName=lowerarm_r, InputRange=(X=0,Y=1), TwistAxis="1,0,0").RawOutput) -> (Output_Get:double)
   Set upperarm_up_l(
    execute=@n4.then_2
    upperarm_up_l:double=GetAlphaFromReferenceTransformDelta(
     SkeletalMeshComponent=GetOwnerSkeletalMeshComponent()
     BoneName=upperarm_l
     BaseName=clavicle_l
     DeltaComponentType=EulerY
     InputRange=(X=0,Y=1)
     TwistAxis="1,0,0"
    ).RawOutput
   ) -> (Output_Get:double)
   > Set upperarm_up_r(upperarm_up_r:double=GetAlphaFromReferenceTransformDelta(
     SkeletalMeshComponent=GetOwnerSkeletalMeshComponent()
     BoneName=upperarm_r
     BaseName=clavicle_r
     DeltaComponentType=EulerY
     InputRange=(X=0,Y=1)
     TwistAxis="1,0,0"
    ).RawOutput) -> (Output_Get:double)
   Set abs(
    execute=@n4.then_4
    abs:double=GetAlphaFromReferenceTransformDelta(SkeletalMeshComponent=GetOwnerSkeletalMeshComponent(), BoneName=spine_04, BaseName=pelvis, InputRange=(X=0,Y=1), TwistAxis="1,0,0").RawOutput
   ) -> (Output_Get:double)
   Set elbow_rot_l(
    execute=@n4.then_0
    elbow_rot_l:double=GetAlphaFromReferenceTransformDelta(SkeletalMeshComponent=GetOwnerSkeletalMeshComponent(), BoneName=lowerarm_l, BaseName=upperarm_l, InputRange=(X=0,Y=1), TwistAxis="1,0,0").RawOutput
   ) -> (Output_Get:double)
   > Set elbow_rot_r(elbow_rot_r:double=GetAlphaFromReferenceTransformDelta(SkeletalMeshComponent=GetOwnerSkeletalMeshComponent(), BoneName=lowerarm_r, BaseName=upperarm_r, InputRange=(X=0,Y=1), TwistAxis="1,0,0").RawOutput) -> (Output_Get:double)
   Set clavicle_fwdBk_l(
    execute=@n4.then_1
    clavicle_fwdBk_l:double=GetAlphaFromReferenceTransformDelta(SkeletalMeshComponent=GetOwnerSkeletalMeshComponent(), BoneName=clavicle_l, BaseName=spine_05, InputRange=(X=0,Y=1), TwistAxis="1,0,0").RawOutput
   ) -> (Output_Get:double)
   > Set clavicle_fwdBk_r(clavicle_fwdBk_r:double=GetAlphaFromReferenceTransformDelta(SkeletalMeshComponent=GetOwnerSkeletalMeshComponent(), BoneName=clavicle_r, BaseName=spine_05, InputRange=(X=0,Y=1), TwistAxis="1,0,0").RawOutput) -> (Output_Get:double)
   Sequence(execute=@n3.then_1)
   > n5 = Get Dynamic Material Instance(SkeletalMeshComponent=GetPartSkeletalMeshComponent())
   > n6 = Sequence!(execute) -> (then_0, then_1, then_2, then_3, then_4, then_5:Exec, then_6:Exec)
   n7 = MuscleFNC!(
    execute=@n6.then_2
    self
    inTarget:Material Instance Dynamic Object Reference=@n5
    paramName:Name=Deltoid_L
    value:double=Get upperarm_up_l() -> (upperarm_up_l:double)
    inRangeA:double=0
    inRangeB:double=60
    curve:Curve Float Object Reference=Get easyInOutCrv_var() -> (easyInOutCrv_var:Curve Float Object Reference)
    alpha:double=1
   ) -> (then, outValue:double, outTarget:Material Instance Dynamic Object Reference)
   > n8 = MuscleFNC(
    inTarget=@n7.outTarget
    paramName=Deltoid_R
    value=Get upperarm_up_r() -> (upperarm_up_r:double)
    inRangeA=0
    inRangeB=60
    curve=Get easyInOutCrv_var()
    alpha=1
   )
   > n9 = MuscleFNC(
    inTarget=@n8.outTarget
    paramName=Chest L
    value=Get clavicle_fwdBk_l() -> (clavicle_fwdBk_l:double)
    inRangeA=0
    inRangeB=25
    curve=Get easyInOutCrv_var()
    alpha=1
   )
   > MuscleFNC(
    inTarget=@n9.outTarget
    paramName=Chest R
    value=Get clavicle_fwdBk_r() -> (clavicle_fwdBk_r:double)
    inRangeA=0
    inRangeB=25
    curve=Get easyInOutCrv_var()
    alpha=1
   )
   n10 = MuscleFNC(
    execute=@n6.then_0
    paramName=Biceps_L
    value=Get elbow_rot_l() -> (elbow_rot_l:double)
    inRangeA=0
    inRangeB=95
    curve=Get easyInOutCrv_var()
    alpha=1
   )
   > n11 = MuscleFNC(
    inTarget=@n10.outTarget
    paramName=Biceps_R
    value=Get elbow_rot_r() -> (elbow_rot_r:double)
    inRangeA=0
    inRangeB=95
    curve=Get easyInOutCrv_var()
    alpha=1
   )
   > n12 = MuscleFNC(inTarget=@n11.outTarget, paramName=Triceps_L, value=Get elbow_rot_l(), inRangeA=0, inRangeB=-20, curve=Get easyInOutCrv_var(), alpha=1)
   > MuscleFNC(inTarget=@n12.outTarget, paramName=Triceps_R, value=Get elbow_rot_r(), inRangeA=0, inRangeB=-20, curve=Get easyInOutCrv_var(), alpha=1)
   n13 = MuscleFNC(
    execute=@n6.then_4
    inTarget=@n5
    paramName=Back_L
    value=Get clavicle_fwdBk_l()
    inRangeA=0
    inRangeB=30
    curve=Get easyInOutCrv_var()
    alpha=1
   )
   > n14 = MuscleFNC(inTarget=@n13.outTarget, paramName=Back_L, value=Get clavicle_fwdBk_r(), inRangeA=0, inRangeB=-10, curve=Get easyInOutCrv_var(), alpha=1)
   > n15 = MuscleFNC(
    inTarget=@n14.outTarget
    paramName=Forearm_L
    value=Get wrist_bend_l() -> (wrist_bend_l:double)
    inRangeA=0
    inRangeB=-20
    curve=Get easyInOutCrv_var()
    alpha=1
   )
   > MuscleFNC(
    inTarget=@n15.outTarget
    paramName=Forearm_R
    value=Get wrist_bend_r() -> (wrist_bend_r:double)
    inRangeA=0
    inRangeB=-20
    curve=Get easyInOutCrv_var()
    alpha=1
   )
   MuscleFNC(
    execute=@n6.then_6
    inTarget=@n5
    paramName=Abdomen
    value=Get abs() -> (abs:double)
    inRangeA=0
    inRangeB=35
    curve=Get easyInOutCrv_var()
    alpha=1
   )
   n16 = Gravity Override State Function DEPRECATED(
    execute=@n1.then_1
    CharacterPartAnimInstance=Self-Reference()
    GameTimeInSeconds=GetGameTimeInSeconds()
    GravityOverrideParams=Get Gravity Override Params()
   )
   > Set GravityOverride(GravityOverride=@n16)
  ]}
 MuscleFNC {nodes [
   n1 = MuscleFNC!() -> (
    then
    inTarget:Material Instance Dynamic Object Reference
    paramName:Name
    value:double=1
    inRangeA:double=~
    inRangeB:double=~
    curve:Curve Float Object Reference
    alpha:double=1
   ) {is=title K2Node_FunctionEntry}
   n2 = float * float(
    A=GetFloatValue!(self:Curve Float Object Reference=@n1.curve, InTime:float=MapRangeClamped(Value=@n1.value, InRangeA=@n1.inRangeA, InRangeB=@n1.inRangeB)) -> (ReturnValue:float)
    B=@n1.alpha
   ) {is=title K2Node_PromotableOperator}
   > SetScalarParameterValue(self=@n1, ParameterName=@n1.paramName, Value=@n2)
   > Return Node!(execute, outValue:double=@n2, outTarget:Material Instance Dynamic Object Reference=@n1)
  ]}
]
variables [
 abs {type=number}
 clavicle_fwdBk_l {type=number}
 clavicle_fwdBk_r {type=number}
 easyInOutCrv_var {type=CurveFloat, default=&/muscleEasyInOutCrv}
 elbow_rot_l {type=number}
 elbow_rot_r {type=number}
 Gravity Override Params {type=GravityOverrideParamsStruct}
 Gravity Override Params_0 {type=GravityOverrideParamsStruct}
 GravityOverride {type=Vector}
 IsMuscleON {type=boolean}
 upperarm_up_l {type=number}
 upperarm_up_r {type=number}
 wrist_bend_l {type=number}
 wrist_bend_r {type=number}
]
